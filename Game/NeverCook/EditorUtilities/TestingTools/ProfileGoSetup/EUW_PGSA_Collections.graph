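# Editor utility widget that browses ProfileGo collections and their scenarios in two side-by-side lists for testing.
# SetProfileGoData stores both string maps, then splits each collection's comma-separated scenario names and checks them against the scenario map, setting label text and color by how many exist.
# Selecting a collection lists its scenarios, marking missing ones.
# Left-clicking a scenario executes its mapped string as a console command in the game world.
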EditorUtilityWidgetBlueprint /Game/NeverCook/EditorUtilities/TestingTools/ProfileGoSetup/EUW_PGSA_Collections
graphs [
 EventGraph {nodes [
   n1 = "On Item Selection Changed (listCollections)"!() -> (
    OutputDelegate
    then
    selectedData:Array of Object References
    selectedWidgets:Array of User Widget Object References
    selectedIndices:Array of Integers
   ) {is=title K2Node_ComponentBoundEvent}
   > n2 = Cast To TA_GenericListItemData!(execute, Object="Get (a copy)"(Array:Array of Object References=@n1.selectedData) -> (Output:Object Reference)) -> (then, CastFailed, AsTA Generic List Item Data:TA Generic List Item Data Object Reference) {is=title K2Node_DynamicCast}
   > PopulateScenariosFromList!(execute, self, Array:Array of Strings=Get strings!(self:TA Generic List Item Data Object Reference=@n2) -> (strings:Array of Strings)) -> (then)
   n3 = "On Item Left Clicked (listScenarios)"!() -> (OutputDelegate, then, isShiftDown:bool, isCtrlDown:bool, data:Object Reference, item:User Widget Object Reference) {is=title K2Node_ComponentBoundEvent}
   > n4 = Cast To TA_GenericListItemData(Object=@n3.data)
   > n5 = GetGameWorld!(execute, self:Unreal Editor Subsystem Object Reference=Get UnrealEditorSubsystem()) -> (then, ReturnValue:World Object Reference)
   > ExecuteConsoleCommand(WorldContextObject=@n5, Command="Get (a ref)"(Array:Array of Strings=Get strings(self=@n4), Dimension 1=0) -> (Output:"String (by ref)"))
   Event PreConstruct()
   Event Construct()
  ]}
 PopulateCollectionList {nodes [
   PopulateCollectionList() {is=title K2Node_FunctionEntry}
   > n1 = Sequence!(execute) -> (then_0, then_1, then_2:Exec)
   > clearItems(self:TA Generic 01 List Object Reference=[
     Get listScenarios() -> (listScenarios:TA Generic 01 List Object Reference)
     Get listCollections() -> (listCollections:TA Generic 01 List Object Reference)
    ])
   n2 = Get CurrentCollectionData() -> (CurrentCollectionData:Map of Strings to Strings)
   n3 = Keys!(execute=@n1.then_1, TargetMap:Map of Strings to Strings=@n2) -> (then, Keys:Array of Strings)
   > n4 = Values!(execute, TargetMap:Map of Strings to Strings=@n2) -> (then, Values:Array of Strings)
   > n5 = For Each Loop(Array=@n3)
   > Set l_possibleScenarios(l_possibleScenarios:Array of Strings=ParseIntoArray(SourceString="Get (a ref)"(Array=@n4, Dimension 1=@n5.Array Index), Delimiter=",")) -> (Output_Get:Array of Strings)
   > n6 = Sequence!(execute) -> (then_0, then_1)
   > Clear(TargetArray=Get l_finalScenarios() -> (l_finalScenarios:Array of Strings))
   > n7 = For Each Loop(Array=Get l_possibleScenarios() -> (l_possibleScenarios:Array of Strings))
   > Branch(Condition=Find(TargetMap=Get CurrentScenarioData() -> (CurrentScenarioData:Map of Strings to Strings), Key=@n7).ReturnValue)
   > Add(TargetArray=Get l_finalScenarios(), NewItem=@n7) {is=title K2Node_CallArrayFunction}
   n8 = Construct Object from Class(execute=@n6.then_1, Class=/Game/Blueprints/Blutilities/ForTechArt/Data/TA_GenericListItemData.TA_GenericListItemData_C) -> (ReturnValue:TA Generic List Item Data Object Reference)
   > Set strings!(execute, strings:Array of Strings=Get l_possibleScenarios(), self:TA Generic List Item Data Object Reference=@n8) -> (then, Output_Get:Array of Strings)
   n9 = Length(TargetArray:Array of Strings=Get l_possibleScenarios())
   n10 = Length(TargetArray=Get l_finalScenarios())
   n11 = "integer < integer"(A=@n10, B=@n9)
   n12 = SelectString(A=(Missing Some Scenarios), B=Append(A=" - ", B="To String (Integer)"(InInt=@n9)), bPickA=@n11)
   n13 = "integer < integer"(A=@n10, B=1)
   > Set displayString!(
    execute
    displayString:String=Append(A=@n5, B=SelectString(A=(Missing All Scenarios), B=@n12, bPickA=@n13))
    self:TA Generic List Item Data Object Reference=@n8
   ) -> (then, Output_Get:String)
   > n14 = addItem!(execute, self:TA Generic 01 List Object Reference=Get listCollections(), dataObject:Object Reference=@n8) -> (then, newItem:User Widget Object Reference, newData:Object Reference)
   > n15 = Cast To TA_Generic01_ListItem!(execute, Object=@n14) -> (then, CastFailed, AsTA Generic 01 List Item:TA Generic 01 List Item Object Reference) {is=title K2Node_DynamicCast}
   > Set labelColor_Default!(
    execute
    labelColor_Default:Linear Color Structure=SelectColor(
     A=SelectColor(A=(R=0.6573,G=0.0296,B=0.0544,A=1), B=(R=0.2133,G=0.3635,B=1,A=1), bPickA=@n11)
     B=(R=0.6021,G=0.4788,B=0.0412,A=1)
     bPickA=@n13
    )
    self:TA Base Editor Button Widget Object Reference=@n15
   ) -> (then, Output_Get:Linear Color Structure)
   > validateVisuals!(execute, self:TA Base Editor Button Widget Object Reference=@n15) -> (then, " ":bool)
  ]}
 PopulateScenariosFromList {nodes [
   n1 = PopulateScenariosFromList!() -> (then, Array:Array of Strings) {is=title K2Node_FunctionEntry}
   > clearItems(self=Get listScenarios())
   > n2 = For Each Loop(Array=@n1)
   > n3 = Construct Object from Class(Class=&TA_GenericListItemData_C)
   n4 = Find(TargetMap=Get CurrentScenarioData(), Key=@n2)
   > Set strings(strings=Make Array("[0]":String=@n4) -> (Array:Array of Strings), self=@n3)
   > Set displayString(displayString=Append(A=@n2, B=SelectString(B=(Missing), bPickA=@n4.ReturnValue)), self=@n3)
   > addItem(self=Get listScenarios(), dataObject=@n3)
  ]}
 SetProfileGoData {nodes [
   n1 = SetProfileGoData!() -> (then, CollectionData:Map of Strings to Strings, ScenarioData:Map of Strings to Strings) {is=title K2Node_FunctionEntry}
   > Set CurrentCollectionData(CurrentCollectionData:Map of Strings to Strings=@n1) -> (Output_Get:Map of Strings to Strings)
   > Set CurrentScenarioData(CurrentScenarioData:Map of Strings to Strings=@n1.ScenarioData) -> (Output_Get:Map of Strings to Strings)
   > PopulateCollectionList() {is=title K2Node_CallFunction}
  ]}
]
variables [
 CurrentCollectionData {type="map<string>", default {}}
 CurrentScenarioData {type="map<string>", default {}}
]
widgets {
 info {rootWidgetClass=&StackBox, widgetCount=12}
 widgets [
  StackBox_0 {widgetClassPath=&StackBox, widget=StackBox_0, properties {orientation=Orient_Vertical}}
  Border_0 {
   widgetClassPath=&Border
   parent=StackBox_0
   widget=Border_0
   properties {brushColor=<rgba 0 0 0 0.5>, padding=<ltrb 10 10 10 10>}
   slot {class=StackBoxSlot, properties {size {sizeRule=Fill, value=1}}}
  }
  VerticalBox_0 {
   widgetClassPath=&VerticalBox
   parent=Border_0
   widget=VerticalBox_0
   slot {class=BorderSlot, properties {padding=<ltrb 10 10 10 10>}}
  }
  HorizontalBox_1 {widgetClassPath=&HorizontalBox, parent=VerticalBox_0, widget=HorizontalBox_1, slot {class=VerticalBoxSlot}}
  CommonTextBlock_94 {
   widgetClassPath=&CommonTextBlock
   parent=HorizontalBox_1
   widget=CommonTextBlock_94
   properties {
    justification=Center
    text=COLLECTIONS
    font {
     bForceMonospaced=false
     bMaterialIsStencil=false
     fontMaterial=null
     fontObject=&NotoSans
     letterSpacing=0
     monospacedWidth=1
     size=12
     skewAmount=0
     typefaceFontName=NotoSans-Bold
     outlineSettings {
      bApplyOutlineToDropShadows=false
      bMiteredCorners=false
      bSeparateFillAlpha=false
      outlineMaterial=null
      outlineSize=0
      outlineColor=<rgba 0 0 0 1>
     }
    }
    shadowOffset=<0 0>
   }
   slot {class=HorizontalBoxSlot, properties {size {sizeRule=Fill, value=1}}}
  }
  CommonTextBlock {
   widgetClassPath=&CommonTextBlock
   parent=HorizontalBox_1
   widget=CommonTextBlock
   properties {
    justification=Center
    text=Scenarios
    font {
     bForceMonospaced=false
     bMaterialIsStencil=false
     fontMaterial=null
     fontObject=&NotoSans
     letterSpacing=0
     monospacedWidth=1
     size=12
     skewAmount=0
     typefaceFontName=NotoSans-Bold
     outlineSettings {
      bApplyOutlineToDropShadows=false
      bMiteredCorners=false
      bSeparateFillAlpha=false
      outlineMaterial=null
      outlineSize=0
      outlineColor=<rgba 0 0 0 1>
     }
    }
    shadowOffset=<0 0>
   }
   slot {class=HorizontalBoxSlot, properties {size {sizeRule=Fill, value=1}}}
  }
  Border_1 {
   widgetClassPath=&Border
   parent=VerticalBox_0
   widget=Border_1
   properties {brushColor=<rgba 0.6146 0.6146 0.6146 1>, padding=<ltrb 3 3 3 3>}
   slot {class=VerticalBoxSlot, properties {size {sizeRule=Fill, value=1}}}
  }
  Border_2 {
   widgetClassPath=&Border
   parent=Border_1
   widget=Border_2
   properties {brushColor=<rgba 0.0323 0.0323 0.0323 1>, padding=<ltrb 0 0 0 0>}
   slot {class=BorderSlot, properties {padding=<ltrb 3 3 3 3>}}
  }
  HorizontalBox_0 {
   widgetClassPath=&HorizontalBox
   parent=Border_2
   widget=HorizontalBox_0
   slot {class=BorderSlot, properties {padding=<ltrb 0 0 0 0>}}
  }
  listCollections {
   widgetClassPath=/Game/Blueprints/Blutilities/ForTechArt/Widgets/List/TA_Generic01_List
   bIsVariable=true
   parent=HorizontalBox_0
   widget=listCollections
   slot {class=HorizontalBoxSlot, properties {size {sizeRule=Fill, value=1}}}
  }
  Image {
   widgetClassPath=&Image
   bIsVariable=true
   parent=HorizontalBox_0
   widget=Image
   properties {
    brush {
     bIsDynamicallyLoaded=false
     drawAs=Image
     imageType=NoImage
     mirroring=NoMirror
     resourceName=None
     resourceObject=null
     tiling=NoTile
     imageSize=<3 3>
     margin=<ltrb 0 0 0 0>
     outlineSettings {
      bUseBrushTransparency=false
      roundingType=HalfHeightRadius
      width=0
      color {colorUseRule=UseColor_Specified, specifiedColor=<rgba 0 0 0 0>}
      cornerRadii=<0 0 0 0>
     }
     tintColor {colorUseRule=UseColor_Specified, specifiedColor=<rgba 1 1 1 1>}
     uVRegion {bIsValid=false, max=<0 0>, min=<0 0>}
    }
    colorAndOpacity=<rgba 0.6146 0.6146 0.6146 1>
   }
   slot {class=HorizontalBoxSlot}
  }
  listScenarios {
   widgetClassPath=&TA_Generic01_List
   bIsVariable=true
   parent=HorizontalBox_0
   widget=listScenarios
   slot {class=HorizontalBoxSlot, properties {size {sizeRule=Fill, value=1}}}
  }
 ]
}
