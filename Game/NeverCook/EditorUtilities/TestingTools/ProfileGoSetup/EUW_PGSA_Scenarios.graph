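# Editor utility widget for ProfileGo setup that audits scenarios against collections in three lists: used, used-but-missing and unused.
# SetProfileGoData stores collection and scenario maps, then PopulateCollectionList splits comma-separated collection values and compares them to scenario keys with set intersection and differences.
# List items carry the scenario name and map value; left-click on used or unused runs that value as a console command (bugit go-to), right-click copies the name.
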EditorUtilityWidgetBlueprint /Game/NeverCook/EditorUtilities/TestingTools/ProfileGoSetup/EUW_PGSA_Scenarios
graphs [
 EventGraph {nodes [
   n1 = "On Item Left Clicked (listUsed)"!() -> (OutputDelegate, then, isShiftDown:bool, isCtrlDown:bool, data:Object Reference, item:User Widget Object Reference) {is=title K2Node_ComponentBoundEvent}
   > HandleListGoTo!(execute, self, ListData:Object Reference=@n1.data) -> (then)
   n2 = "On Item Left Clicked (listUnused)"!() -> (OutputDelegate, then, isShiftDown:bool, isCtrlDown:bool, data:Object Reference, item:User Widget Object Reference) {is=title K2Node_ComponentBoundEvent}
   > HandleListGoTo(ListData=@n2.data)
   n3 = "On Item Right Clicked (listMissing)"!() -> (OutputDelegate, then, isShiftDown:bool, isCtrlDown:bool, data:Object Reference, item:User Widget Object Reference) {is=title K2Node_ComponentBoundEvent}
   > HandleListCopyName!(execute, self, ListData:Object Reference=@n3.data) -> (then)
   n4 = "On Item Right Clicked (listUnused)"!() -> (OutputDelegate, then, isShiftDown:bool, isCtrlDown:bool, data:Object Reference, item:User Widget Object Reference) {is=title K2Node_ComponentBoundEvent}
   > HandleListCopyName(ListData=@n4.data)
   n5 = "On Item Right Clicked (listUsed)"!() -> (OutputDelegate, then, isShiftDown:bool, isCtrlDown:bool, data:Object Reference, item:User Widget Object Reference) {is=title K2Node_ComponentBoundEvent}
   > HandleListCopyName(ListData=@n5.data)
  ]}
 HandleListCopyName {nodes [
   n1 = HandleListCopyName!() -> (then, ListData:Object Reference) {is=title K2Node_FunctionEntry}
   > n2 = Cast To TA_GenericListItemData!(execute, Object=@n1) -> (then, CastFailed, AsTA Generic List Item Data:TA Generic List Item Data Object Reference) {is=title K2Node_DynamicCast}
   > CopyStringToClipboard!(execute, InString:String=Get displayString!(self:TA Generic List Item Data Object Reference=@n2) -> (displayString:String)) -> (then)
  ]}
 HandleListGoTo {nodes [
   n1 = HandleListGoTo!() -> (then, ListData:Object Reference) {is=title K2Node_FunctionEntry}
   > n2 = Cast To TA_GenericListItemData(Object=@n1)
   > n3 = GetGameWorld!(execute, self:Unreal Editor Subsystem Object Reference=Get UnrealEditorSubsystem()) -> (then, ReturnValue:World Object Reference)
   > ExecuteConsoleCommand(
    WorldContextObject=@n3
    Command="Get (a ref)"(Array:Array of Strings=Get strings!(self:TA Generic List Item Data Object Reference=@n2) -> (strings:Array of Strings), Dimension 1=0) -> (Output:"String (by ref)")
   )
  ]}
 PopulateCollectionList {nodes [
   PopulateCollectionList() {is=title K2Node_FunctionEntry}
   > n1 = Sequence!(execute) -> (then_0, then_1, then_2:Exec, then_3:Exec, then_4:Exec, then_5:Exec)
   > clearItems(self:TA Generic 01 List Object Reference=[
     Get listMissing() -> (listMissing:TA Generic 01 List Object Reference)
     Get listUsed() -> (listUsed:TA Generic 01 List Object Reference)
     Get listUnused() -> (listUnused:TA Generic 01 List Object Reference)
    ])
   n2 = Values!(execute=@n1.then_1, TargetMap:Map of Strings to Strings=Get CurrentCollectionData() -> (CurrentCollectionData:Map of Strings to Strings)) -> (then, Values:Array of Strings)
   > n3 = For Each Loop(Array=@n2)
   > Add Items!(
    execute
    TargetSet:Set of Strings=Get l_collectionScenarios() -> (l_collectionScenarios:Set of Strings)
    NewItems:Array of Strings=ParseIntoArray(SourceString=@n3, Delimiter=",")
   ) -> (then)
   n4 = Keys!(execute=@n1.then_2, TargetMap:Map of Strings to Strings=Get CurrentScenarioData() -> (CurrentScenarioData:Map of Strings to Strings)) -> (then, Keys:Array of Strings)
   > Add Items(TargetSet=Get l_existingScenarios() -> (l_existingScenarios:Set of Strings), NewItems=@n4)
   n5 = Difference!(execute=@n1.then_3, A:Set of Strings=Get l_existingScenarios(), B:Set of Strings=Get l_collectionScenarios()) -> (then, Result:Set of Strings)
   > n6 = Set l_unusedScenarios(l_unusedScenarios:Set of Strings=@n5) -> (Output_Get:Set of Strings)
   > n7 = To Array!(execute, A:Set of Strings=@n6) -> (then, Result:Array of Strings)
   > PopulateListFromStrings!(execute, self, scenarioStrings:Array of Strings=@n7, list:TA Generic 01 List Object Reference=Get listUnused()) -> (then)
   n8 = Intersection!(execute=@n1.then_4, A:Set of Strings=Get l_collectionScenarios(), B:Set of Strings=Get l_existingScenarios()) -> (then, Result:Set of Strings)
   > n9 = Set l_usedScenarios(l_usedScenarios:Set of Strings=@n8) -> (Output_Get:Set of Strings)
   > n10 = To Array(A=@n9)
   > PopulateListFromStrings(scenarioStrings=@n10, list=Get listUsed())
   n11 = Difference(execute=@n1.then_5, A=Get l_collectionScenarios(), B=Get l_existingScenarios())
   > n12 = Set l_missingScenarios(l_missingScenarios:Set of Strings=@n11) -> (Output_Get:Set of Strings)
   > n13 = To Array(A=@n12)
   > PopulateListFromStrings(scenarioStrings=@n13, list=Get listMissing())
  ]}
 PopulateListFromStrings {nodes [
   n1 = PopulateListFromStrings!() -> (then, scenarioStrings:Array of Strings, list:TA Generic 01 List Object Reference) {is=title K2Node_FunctionEntry}
   > clearItems(self=@n1.list)
   > n2 = For Each Loop(Array=@n1)
   > n3 = Construct Object from Class(Class=/Game/Blueprints/Blutilities/ForTechArt/Data/TA_GenericListItemData.TA_GenericListItemData_C) -> (ReturnValue:TA Generic List Item Data Object Reference)
   n4 = Find(TargetMap=Get CurrentScenarioData(), Key=@n2)
   > n5 = Branch(Condition=@n4.ReturnValue)
   > n6 = Set strings!(
    execute
    strings:Array of Strings=Make Array("[0]":String=@n4) -> (Array:Array of Strings)
    self:TA Generic List Item Data Object Reference=@n3
   ) -> (then, Output_Get:Array of Strings)
   Set displayString!(execute=[@n5.else, @n6.then], displayString:String=@n2, self:TA Generic List Item Data Object Reference=@n3) -> (then, Output_Get:String)
   > addItem!(execute, self:TA Generic 01 List Object Reference=Get list() -> (list:TA Generic 01 List Object Reference), dataObject:Object Reference=@n3) -> (then, newItem:User Widget Object Reference, newData:Object Reference)
  ]}
 PopulateScenariosFromList {nodes [
   n1 = PopulateScenariosFromList!() -> (then, Array:Array of Strings) {is=title K2Node_FunctionEntry}
   > clearItems(self=Get listMissing())
   > n2 = For Each Loop(Array=@n1)
   > n3 = Construct Object from Class(Class=&TA_GenericListItemData_C)
   n4 = Find(TargetMap=Get CurrentScenarioData(), Key=@n2)
   > Set strings(strings=Make Array("[0]"=@n4), self=@n3)
   > Set displayString(displayString=Append(A=@n2, B=SelectString(B=(Missing), bPickA=@n4.ReturnValue)), self=@n3)
   > addItem(self=Get listMissing(), dataObject=@n3)
  ]}
 SetProfileGoData {nodes [
   n1 = SetProfileGoData!() -> (then, CollectionData:Map of Strings to Strings, ScenarioData:Map of Strings to Strings) {is=title K2Node_FunctionEntry}
   > Set CurrentCollectionData(CurrentCollectionData:Map of Strings to Strings=@n1) -> (Output_Get:Map of Strings to Strings)
   > Set CurrentScenarioData(CurrentScenarioData:Map of Strings to Strings=@n1.ScenarioData) -> (Output_Get:Map of Strings to Strings)
   > PopulateCollectionList() {is=title K2Node_CallFunction}
  ]}
]
variables [
 CurrentCollectionData {type="map<string>", default {}}
 CurrentScenarioData {type="map<string>", default {}}
]
widgets {
 info {rootWidgetClass=&StackBox, widgetCount=15}
 widgets [
  StackBox_0 {widgetClassPath=&StackBox, widget=StackBox_0, properties {orientation=Orient_Vertical}}
  Border_0 {
   widgetClassPath=&Border
   parent=StackBox_0
   widget=Border_0
   properties {brushColor=<rgba 0 0 0 0.5>, padding=<ltrb 10 10 10 10>}
   slot {class=StackBoxSlot, properties {size {sizeRule=Fill, value=1}}}
  }
  VerticalBox_0 {
   widgetClassPath=&VerticalBox
   parent=Border_0
   widget=VerticalBox_0
   slot {class=BorderSlot, properties {padding=<ltrb 10 10 10 10>}}
  }
  HorizontalBox_1 {widgetClassPath=&HorizontalBox, parent=VerticalBox_0, widget=HorizontalBox_1, slot {class=VerticalBoxSlot}}
  txtUsed {
   widgetClassPath=&CommonTextBlock
   parent=HorizontalBox_1
   widget=txtUsed
   properties {
    justification=Center
    text=Used Valid Scenarios
    toolTipText=Scenarios that exist AND are used in a Collection.
    font {
     bForceMonospaced=false
     bMaterialIsStencil=false
     fontMaterial=null
     fontObject=&NotoSans
     letterSpacing=0
     monospacedWidth=1
     size=12
     skewAmount=0
     typefaceFontName=NotoSans-Bold
     outlineSettings {
      bApplyOutlineToDropShadows=false
      bMiteredCorners=false
      bSeparateFillAlpha=false
      outlineMaterial=null
      outlineSize=0
      outlineColor=<rgba 0 0 0 1>
     }
    }
    shadowOffset=<0 0>
   }
   slot {class=HorizontalBoxSlot, properties {size {sizeRule=Fill, value=1}}}
  }
  txtMissing {
   widgetClassPath=&CommonTextBlock
   parent=HorizontalBox_1
   widget=txtMissing
   properties {
    justification=Center
    text=Used Missing Scenarios
    toolTipText=Scenarios that are trying to be used by a Collection but do not actually exist.
    font {
     bForceMonospaced=false
     bMaterialIsStencil=false
     fontMaterial=null
     fontObject=&NotoSans
     letterSpacing=0
     monospacedWidth=1
     size=12
     skewAmount=0
     typefaceFontName=NotoSans-Bold
     outlineSettings {
      bApplyOutlineToDropShadows=false
      bMiteredCorners=false
      bSeparateFillAlpha=false
      outlineMaterial=null
      outlineSize=0
      outlineColor=<rgba 0 0 0 1>
     }
    }
    shadowOffset=<0 0>
   }
   slot {class=HorizontalBoxSlot, properties {size {sizeRule=Fill, value=1}}}
  }
  txtUnused {
   widgetClassPath=&CommonTextBlock
   parent=HorizontalBox_1
   widget=txtUnused
   properties {
    justification=Center
    text=Unused Valid Scenarios
    toolTipText=Scenarios that exist but no Collection references them.
    font {
     bForceMonospaced=false
     bMaterialIsStencil=false
     fontMaterial=null
     fontObject=&NotoSans
     letterSpacing=0
     monospacedWidth=1
     size=12
     skewAmount=0
     typefaceFontName=NotoSans-Bold
     outlineSettings {
      bApplyOutlineToDropShadows=false
      bMiteredCorners=false
      bSeparateFillAlpha=false
      outlineMaterial=null
      outlineSize=0
      outlineColor=<rgba 0 0 0 1>
     }
    }
    shadowOffset=<0 0>
   }
   slot {class=HorizontalBoxSlot, properties {size {sizeRule=Fill, value=1}}}
  }
  Border_1 {
   widgetClassPath=&Border
   parent=VerticalBox_0
   widget=Border_1
   properties {brushColor=<rgba 0.6146 0.6146 0.6146 1>, padding=<ltrb 3 3 3 3>}
   slot {class=VerticalBoxSlot, properties {size {sizeRule=Fill, value=1}}}
  }
  Border_2 {
   widgetClassPath=&Border
   parent=Border_1
   widget=Border_2
   properties {brushColor=<rgba 0.0323 0.0323 0.0323 1>, padding=<ltrb 0 0 0 0>}
   slot {class=BorderSlot, properties {padding=<ltrb 3 3 3 3>}}
  }
  HorizontalBox_0 {
   widgetClassPath=&HorizontalBox
   parent=Border_2
   widget=HorizontalBox_0
   slot {class=BorderSlot, properties {padding=<ltrb 0 0 0 0>}}
  }
  listUsed {
   widgetClassPath=/Game/Blueprints/Blutilities/ForTechArt/Widgets/List/TA_Generic01_List
   bIsVariable=true
   parent=HorizontalBox_0
   widget=listUsed
   properties {toolTipText="Left Click -> Go to bugit coordinates in level viewport\r\n\r\nRight Click -> Copy name to clipboard"}
   slot {class=HorizontalBoxSlot, properties {size {sizeRule=Fill, value=1}}}
  }
  Image {
   widgetClassPath=&Image
   bIsVariable=true
   parent=HorizontalBox_0
   widget=Image
   properties {
    brush {
     bIsDynamicallyLoaded=false
     drawAs=Image
     imageType=NoImage
     mirroring=NoMirror
     resourceName=None
     resourceObject=null
     tiling=NoTile
     imageSize=<3 3>
     margin=<ltrb 0 0 0 0>
     outlineSettings {
      bUseBrushTransparency=false
      roundingType=HalfHeightRadius
      width=0
      color {colorUseRule=UseColor_Specified, specifiedColor=<rgba 0 0 0 0>}
      cornerRadii=<0 0 0 0>
     }
     tintColor {colorUseRule=UseColor_Specified, specifiedColor=<rgba 1 1 1 1>}
     uVRegion {bIsValid=false, max=<0 0>, min=<0 0>}
    }
    colorAndOpacity=<rgba 0.6146 0.6146 0.6146 1>
   }
   slot {class=HorizontalBoxSlot}
  }
  listMissing {
   widgetClassPath=&TA_Generic01_List
   bIsVariable=true
   parent=HorizontalBox_0
   widget=listMissing
   properties {toolTipText="Right Click -> Copy name to clipboard"}
   slot {class=HorizontalBoxSlot, properties {size {sizeRule=Fill, value=1}}}
  }
  Image_1 {
   widgetClassPath=&Image
   bIsVariable=true
   parent=HorizontalBox_0
   widget=Image_1
   properties {
    brush {
     bIsDynamicallyLoaded=false
     drawAs=Image
     imageType=NoImage
     mirroring=NoMirror
     resourceName=None
     resourceObject=null
     tiling=NoTile
     imageSize=<3 3>
     margin=<ltrb 0 0 0 0>
     outlineSettings {
      bUseBrushTransparency=false
      roundingType=HalfHeightRadius
      width=0
      color {colorUseRule=UseColor_Specified, specifiedColor=<rgba 0 0 0 0>}
      cornerRadii=<0 0 0 0>
     }
     tintColor {colorUseRule=UseColor_Specified, specifiedColor=<rgba 1 1 1 1>}
     uVRegion {bIsValid=false, max=<0 0>, min=<0 0>}
    }
    colorAndOpacity=<rgba 0.6146 0.6146 0.6146 1>
   }
   slot {class=HorizontalBoxSlot}
  }
  listUnused {
   widgetClassPath=&TA_Generic01_List
   bIsVariable=true
   parent=HorizontalBox_0
   widget=listUnused
   properties {toolTipText="Left Click -> Go to bugit coordinates in level viewport\r\n\r\nRight Click -> Copy name to clipboard"}
   slot {class=HorizontalBoxSlot, properties {size {sizeRule=Fill, value=1}}}
  }
 ]
}
