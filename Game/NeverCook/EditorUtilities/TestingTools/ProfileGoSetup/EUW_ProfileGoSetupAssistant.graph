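# Editor utility widget for setting up ProfileGo, showing collections and scenarios from DefaultGame.ini or a picked INI in two tabs.
# A Python helper parses the INI into maps passed to the Collections and Scenarios sub-widgets.
# A Python tkinter dialog picks the INI, whose project-relative path persists through the base widget's save game.
# Tick turns the editor camera transform into a copyable +ProfileGoScenarios config line, with roll fixed at zero.
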EditorUtilityWidgetBlueprint /Game/NeverCook/EditorUtilities/TestingTools/ProfileGoSetup/EUW_ProfileGoSetupAssistant
graphs [
 EventGraph {nodes [
   "On Clicked (TA_Generic01_Button)"!() -> (OutputDelegate, then, widget:User Widget Object Reference, event:Pointer Event Structure) {is=title K2Node_ComponentBoundEvent}
   > LoadProfileGoData()
   "On Clicked (btnSwapIni)"!() -> (OutputDelegate, then, widget:User Widget Object Reference, event:Pointer Event Structure) {is=title K2Node_ComponentBoundEvent}
   n1 = Get CurrentIniPath() -> (CurrentIniPath:String)
   > PrintString(InString=@n1)
   n2 = GetProjectDirectory!() -> (ReturnValue:String)
   > n3 = Execute Python Script!(
    execute
    PythonScript="import unreal\r\nfrom tkinter import Tk\r\nfrom tkinter import filedialog\r\n\r\nroot = Tk()\r\nroot.withdraw()\r\nfile_path = filedialog.askopenfilename(filetypes = ( (\"INI files\",\"*.ini\"), (\"All files\", \"*.*\")),initialdir = in_dir)"
    in_dir:String=Select(
     Option 0:String=ConvertRelativePathToFull!(InPath:String=@n1, InBasePath:String=@n2) -> (ReturnValue:String)
     Option 1:String=@n2
     Index=IsEmpty(InString=@n1)
    ) -> (ReturnValue:String)
   ) -> (then, ReturnValue, file_path:String) {inputs [in_dir], outputs [file_path]}
   > n4 = Branch(Condition=IsEmpty(InString=@n3.file_path))
   Set CurrentIniPath(
    execute=@n4.else
    CurrentIniPath:String=MakePathRelativeTo!(InPath:String=@n3.file_path, InRelativeTo:String=GetProjectDirectory()) -> (OutPath:String, ReturnValue:bool)
   ) -> (Output_Get:String)
   > SaveGame_Save(self:TA Base Editor Widget Object Reference)
   > GetRunButtonLabel()
   "On Clicked (btnRevert)"!() -> (OutputDelegate, then, widget:User Widget Object Reference, event:Pointer Event Structure) {is=title K2Node_ComponentBoundEvent}
   > Set CurrentIniPath()
   > SaveGame_Save()
   > GetRunButtonLabel()
   "On Clicked (btnCopyOutput)"!() -> (OutputDelegate, then, widget:User Widget Object Reference, event:Pointer Event Structure) {is=title K2Node_ComponentBoundEvent}
   n5 = Get editTextOutput() -> (editTextOutput:Editor Utility Editable Text Box Object Reference)
   > CopyStringToClipboard!(execute, InString:String="To String (Text)"(InText="GetText (Text Box)"!(self:Text Box Object Reference=@n5) -> (ReturnValue:Text))) -> (then)
   n6 = "On Item Selection Changed (listTabs)"!() -> (
    OutputDelegate
    then
    selectedData:Array of Object References
    selectedWidgets:Array of User Widget Object References
    selectedIndices:Array of Integers
   ) {is=title K2Node_ComponentBoundEvent}
   > SetActiveWidgetIndex(
    self=Get WidgetSwitcher_0() -> (WidgetSwitcher_0:Widget Switcher Object Reference)
    Index="Max (Integer)"(A="Get (a copy)"(Array:Array of Integers=@n6.selectedIndices) -> (Output:int), B=0)
   )
   Event Construct()
   > Parent: Construct()
   > setIndexSelected!(
    execute
    self:TA Generic 01 List Object Reference=Get listTabs() -> (listTabs:TA Generic 01 List Object Reference)
    index:int=0
    selected:bool=true
    throwEvent:bool=true
   ) -> (then)
   Event Tick()
   > n7 = Get Editor Camera Info!(execute) -> (then, CameraLocation:Vector, ForwardDirection:Vector, FOV:float, ReturnValue:bool) {is=title K2Node_CallFunction}
   n8 = BreakVector(InVec=@n7)
   n9 = BreakRotator(InRot=MakeRotFromX!(X:"Vector (by ref)"=@n7.ForwardDirection) -> (ReturnValue:Rotator))
   n10 = "GetText (Text Box)"(self=Get editTextName() -> (editTextName:Editor Utility Editable Text Box Object Reference))
   n11 = TextIsEmpty(InText=@n10)
   > "SetText (Text Box)"!(
    execute
    self:Text Box Object Reference=Get editTextOutput()
    InText:Text=Format Text!(
     Format="LOCTEXT(\"+ProfileGoScenarios=(Name=\\\"{ScenarioName}\\\",Position=(X={PosX},Y={PosY},Z={PosZ}),orientation=(pitch={Pitch},Yaw={Yaw},Roll={Roll}))\")"
     ScenarioName:Text=Select(Option 0:Text=@n10, Option 1:Text="LOCTEXT(\"NameGoesHere\")", Index=@n11) -> (ReturnValue:Text)
     PosX:Text="To Text (Float)"(Value=@n8, bUseGrouping=false, MinimumIntegralDigits=1, MaximumIntegralDigits=324, MinimumFractionalDigits=0, MaximumFractionalDigits=3)
     PosY:Text="To Text (Float)"(Value=@n8.Y, bUseGrouping=false, MinimumIntegralDigits=1, MaximumIntegralDigits=324, MinimumFractionalDigits=0, MaximumFractionalDigits=3)
     PosZ:Text="To Text (Float)"(Value=@n8.Z, bUseGrouping=false, MinimumIntegralDigits=1, MaximumIntegralDigits=324, MinimumFractionalDigits=0, MaximumFractionalDigits=3)
     Pitch:Text="To Text (Float)"(
      Value=@n9.Pitch
      bUseGrouping=false
      MinimumIntegralDigits=1
      MaximumIntegralDigits=324
      MinimumFractionalDigits=0
      MaximumFractionalDigits=3
     )
     Yaw:Text="To Text (Float)"(Value=@n9.Yaw, bUseGrouping=false, MinimumIntegralDigits=1, MaximumIntegralDigits=324, MinimumFractionalDigits=0, MaximumFractionalDigits=3)
     Roll:Text="To Text (Float)"(Value=0, bUseGrouping=false, MinimumIntegralDigits=1, MaximumIntegralDigits=324, MinimumFractionalDigits=0, MaximumFractionalDigits=3)
    ) -> (Result)
   ) -> (then)
   Event PreConstruct()
   > clearItems(self:TA Generic 01 List Object Reference=Get listTabs())
   > addItemGeneric!(execute, self:TA Generic 01 List Object Reference=Get listTabs(), displayString:String=Collections) -> (then, widget:User Widget Object Reference, data:Object Reference)
   > addItemGeneric(self=Get listTabs(), displayString=Scenarios)
   GetProjectDirectory()
   GetProjectContentDirectory!() -> (ReturnValue:String)
  ]}
 GetRunButtonLabel {nodes [
   GetRunButtonLabel() {is=title K2Node_FunctionEntry}
   n1 = Get CurrentIniPath()
   n2 = IsEmpty(InString=@n1)
   n3 = Format Text!(Format="LOCTEXT(\"Load Data From: \\r\\n{ini}\")", ini:String=@n1) -> (Result)
   > setLabel!(
    execute
    self:TA Base Editor Button Widget Object Reference=Get TA_Generic01_Button() -> (TA_Generic01_Button:TA Generic 01 Button Object Reference)
    label:String="To String (Text)"(InText=Select(Option 0=@n3, Option 1="LOCTEXT(\"Load Data From:\\r\\nDefaultGame.ini\")", Index=@n2))
   ) -> (then)
   GetCleanFilename!(InPath:String=@n1) -> (ReturnValue:String)
  ]}
 LoadProfileGoData {nodes [
   LoadProfileGoData() {is=title K2Node_FunctionEntry}
   n1 = Get CurrentIniPath()
   n2 = GetProjectDirectory()
   > n3 = Execute Python Script!(
    execute
    PythonScript="import unreal\r\nimport TechArt.Tools.ProfileGoTools.TA_ProfileGoHelpers as pgo_helpers\r\nfrom importlib import reload\r\nreload(pgo_helpers)\r\n\r\ncollection_data, scenario_data = pgo_helpers.getLocationDataForAllCollections(file_path)"
    file_path:String=Select(Option 0:String=ConvertRelativePathToFull(InPath=@n1, InBasePath=@n2), Option 1:String, Index=IsEmpty(InString=@n1)) -> (ReturnValue:String)
   ) -> (then, ReturnValue, collection_data:Map of Strings to Strings, scenario_data:Map of Strings to Strings) {inputs [file_path], outputs [collection_data, scenario_data]}
   > Set CurrentCollectionData(CurrentCollectionData:Map of Strings to Strings=@n3.collection_data) -> (Output_Get:Map of Strings to Strings)
   > Set CurrentScenarioData(CurrentScenarioData:Map of Strings to Strings=@n3.scenario_data) -> (Output_Get:Map of Strings to Strings)
   n4 = Get CurrentCollectionData() -> (CurrentCollectionData:Map of Strings to Strings)
   n5 = Get CurrentScenarioData() -> (CurrentScenarioData:Map of Strings to Strings)
   > SetProfileGoData!(
    execute
    self:EUW PGSA Collections Object Reference=Get EUW_PGSA_Collections() -> (EUW_PGSA_Collections:EUW PGSA Collections Object Reference)
    CollectionData:Map of Strings to Strings=@n4
    ScenarioData:Map of Strings to Strings=@n5
   ) -> (then)
   > SetProfileGoData(
    self:EUW PGSA Scenarios Object Reference=Get EUW_PGSA_Scenarios() -> (EUW_PGSA_Scenarios:EUW PGSA Scenarios Object Reference)
    CollectionData=@n4
    ScenarioData=@n5
   )
  ]}
 SaveGame_PostLoad {nodes [
   n1 = SaveGame_PostLoad!() -> (then, saveGameExisted:bool, saveGame_Object:Save Game Object Reference) {is=title K2Node_FunctionEntry}
   > Branch(Condition=@n1)
   > n2 = Cast To TA_Base_SaveGame!(execute, Object=@n1.saveGame_Object) -> (then, CastFailed, AsTA Base Save Game:TA Base Save Game Object Reference) {is=title K2Node_DynamicCast}
   > Set CurrentIniPath(CurrentIniPath="Get (a copy)"(Array:Array of Strings=Get strings!(self:TA Base Save Game Object Reference=@n2) -> (strings:Array of Strings)) -> (Output:String))
   > GetRunButtonLabel() {is=title K2Node_CallFunction}
   Return Node!(execute, " ":bool=false)
  ]}
 SaveGame_PrepareToSave {nodes [
   n1 = SaveGame_PrepareToSave!() -> (then, SaveGameObject:Save Game Object Reference) {is=title K2Node_FunctionEntry}
   > n2 = Cast To TA_Base_SaveGame(Object=@n1)
   > Set strings!(
    execute
    strings:Array of Strings=Make Array("[0]":String=Get CurrentIniPath()) -> (Array:Array of Strings)
    self:TA Base Save Game Object Reference=@n2
   ) -> (then, Output_Get:Array of Strings)
   Return Node(" "=false)
  ]}
]
variables [
 CurrentCollectionData {type="map<string>", default {}}
 CurrentIniPath {type=string, default=""}
 CurrentScenarioData {type="map<string>", default {}}
]
widgets {
 info {
  parentClass=/Game/Blueprints/Blutilities/ForTechArt/Widgets/TA_Base_EditorWidget.TA_Base_EditorWidget_C
  rootWidgetClass=&StackBox
  widgetCount=16
 }
 widgets [
  StackBox_0 {widgetClassPath=&StackBox, widget=StackBox_0, properties {orientation=Orient_Vertical}}
  Border_1 {widgetClassPath=&Border, parent=StackBox_0, widget=Border_1, properties {brushColor=<rgba 0 0 0 0.25>}, slot {class=StackBoxSlot}}
  HorizontalBox_1 {widgetClassPath=&HorizontalBox, parent=Border_1, widget=HorizontalBox_1, slot {class=BorderSlot}}
  btnRevert {
   widgetClassPath=&TA_Generic01_Button
   parent=HorizontalBox_1
   widget=btnRevert
   properties {
    toolTipText=Reset To DefaultGame.ini
    buttonColor_Default=<rgba 0.276 0.0683 0 1>
    buttonStyle_Default {
     clickedSlateSound {resourceObject=null}
     disabled {
      bIsDynamicallyLoaded=false
      drawAs=Box
      imageType=NoImage
      mirroring=NoMirror
      resourceName=None
      resourceObject=/Game/Blueprints/Blutilities/ForTechArt/ArtAssets/Textures/Button02/MI_TA_ButtonBG_02_Down
      tiling=NoTile
      imageSize=<32 32>
      margin=<ltrb 0.25 0.25 0.25 0.25>
      outlineSettings {
       bUseBrushTransparency=false
       roundingType=HalfHeightRadius
       width=0
       color {colorUseRule=UseColor_Specified, specifiedColor=<rgba 0 0 0 0>}
       cornerRadii=<0 0 0 1>
      }
      tintColor {colorUseRule=UseColor_Specified, specifiedColor=<rgba 0.3333 0.3333 0.3333 1>}
      uVRegion {bIsValid=false, max=<0 0>, min=<0 0>}
     }
     disabledForeground {colorUseRule=UseColor_Foreground, specifiedColor=<rgba 1 0 1 1>}
     hovered {
      bIsDynamicallyLoaded=false
      drawAs=Box
      imageType=NoImage
      mirroring=NoMirror
      resourceName=None
      resourceObject=/Game/Blueprints/Blutilities/ForTechArt/ArtAssets/Textures/Button02/MI_TA_ButtonBG_02_Over
      tiling=NoTile
      imageSize=<32 32>
      margin=<ltrb 0.25 0.25 0.25 0.25>
      outlineSettings {
       bUseBrushTransparency=false
       roundingType=HalfHeightRadius
       width=0
       color {colorUseRule=UseColor_Specified, specifiedColor=<rgba 0 0 0 0>}
       cornerRadii=<0 0 0 1>
      }
      tintColor {colorUseRule=UseColor_Specified, specifiedColor=<rgba 0.9 0.9 0.9 1>}
      uVRegion {bIsValid=false, max=<0 0>, min=<0 0>}
     }
     hoveredForeground {colorUseRule=UseColor_Foreground, specifiedColor=<rgba 1 0 1 1>}
     hoveredSlateSound {resourceObject=null}
     normal {
      bIsDynamicallyLoaded=false
      drawAs=Box
      imageType=NoImage
      mirroring=NoMirror
      resourceName=None
      resourceObject=/Game/Blueprints/Blutilities/ForTechArt/ArtAssets/Textures/Button02/M_TA_ButtonBG_02_Up
      tiling=NoTile
      imageSize=<32 32>
      margin=<ltrb 0.25 0.25 0.25 0.25>
      outlineSettings {
       bUseBrushTransparency=false
       roundingType=HalfHeightRadius
       width=0
       color {colorUseRule=UseColor_Specified, specifiedColor=<rgba 0 0 0 0>}
       cornerRadii=<0 0 0 1>
      }
      tintColor {colorUseRule=UseColor_Specified, specifiedColor=<rgba 0.85 0.85 0.85 1>}
      uVRegion {bIsValid=false, max=<0 0>, min=<0 0>}
     }
     normalForeground {colorUseRule=UseColor_Foreground, specifiedColor=<rgba 1 0 1 1>}
     normalPadding=<ltrb 5 5 5 5>
     pressed {
      bIsDynamicallyLoaded=false
      drawAs=Box
      imageType=NoImage
      mirroring=NoMirror
      resourceName=None
      resourceObject=&MI_TA_ButtonBG_02_Down
      tiling=NoTile
      imageSize=<32 32>
      margin=<ltrb 0.25 0.25 0.25 0.25>
      outlineSettings {
       bUseBrushTransparency=false
       roundingType=HalfHeightRadius
       width=0
       color {colorUseRule=UseColor_Specified, specifiedColor=<rgba 0 0 0 0>}
       cornerRadii=<0 0 0 1>
      }
      tintColor {colorUseRule=UseColor_Specified, specifiedColor=<rgba 0.65 0.65 0.65 1>}
      uVRegion {bIsValid=false, max=<0 0>, min=<0 0>}
     }
     pressedForeground {colorUseRule=UseColor_Foreground, specifiedColor=<rgba 1 0 1 1>}
     pressedPadding=<ltrb 5 5 5 5>
     pressedSlateSound {resourceObject=null}
    }
    iconBrush {
     bIsDynamicallyLoaded=false
     drawAs=Image
     imageType=FullColor
     mirroring=NoMirror
     resourceName=None
     resourceObject=/Game/Athena/HUD/Bacchus/HUDLayoutTool/HLT_Icons/T-HLT-Reset
     tiling=NoTile
     imageSize=<20 20>
     margin=<ltrb 0 0 0 0>
     outlineSettings {
      bUseBrushTransparency=false
      roundingType=HalfHeightRadius
      width=0
      color {colorUseRule=UseColor_Specified, specifiedColor=<rgba 0 0 0 0>}
      cornerRadii=<0 0 0 1>
     }
     tintColor {colorUseRule=UseColor_Specified, specifiedColor=<rgba 1 1 1 1>}
     uVRegion {bIsValid=false, max=<0 0>, min=<0 0>}
    }
   }
   slot {class=HorizontalBoxSlot, properties {verticalAlignment=VAlign_Center, padding=<ltrb 10 0 0 0>}}
  }
  btnSwapIni {
   widgetClassPath=&TA_Generic01_Button
   parent=HorizontalBox_1
   widget=btnSwapIni
   properties {label=Pick Game INI, buttonColor_Default=<rgba 0.2711 0.276 0.0398 1>}
   slot {class=HorizontalBoxSlot, properties {verticalAlignment=VAlign_Center, padding=<ltrb 0 0 10 0>}}
  }
  TA_Generic01_Button {
   widgetClassPath=&TA_Generic01_Button
   parent=HorizontalBox_1
   widget=TA_Generic01_Button
   properties {label="Load Data From:\r\nDefaultGame.ini"}
   slot {class=HorizontalBoxSlot, properties {padding=<ltrb 0 10 10 10>, size {sizeRule=Fill, value=1}}}
  }
  HorizontalBox_0 {
   widgetClassPath=&HorizontalBox
   parent=StackBox_0
   widget=HorizontalBox_0
   slot {class=StackBoxSlot, properties {padding=<ltrb 20 10 0 0>}}
  }
  Border_0 {
   widgetClassPath=&Border
   parent=HorizontalBox_0
   widget=Border_0
   properties {brushColor=<rgba 0 0 0 0.25>, padding=<ltrb 7 2 4 2>}
   slot {class=HorizontalBoxSlot}
  }
  listTabs {
   widgetClassPath=/Game/Blueprints/Blutilities/ForTechArt/Widgets/List/TA_Generic01_List
   bIsVariable=true
   parent=Border_0
   widget=listTabs
   properties {allowMultiSelect=false, itemPadding=5, scrollOrientation=Orient_Horizontal}
   slot {class=BorderSlot, properties {padding=<ltrb 7 2 4 2>}}
  }
  WidgetSwitcher_0 {
   widgetClassPath=&WidgetSwitcher
   bIsVariable=true
   parent=StackBox_0
   widget=WidgetSwitcher_0
   slot {class=StackBoxSlot, properties {size {sizeRule=Fill, value=1}}}
  }
  EUW_PGSA_Collections {
   widgetClassPath=&/EUW_PGSA_Collections
   bIsVariable=true
   parent=WidgetSwitcher_0
   widget=EUW_PGSA_Collections
   slot {class=WidgetSwitcherSlot}
  }
  EUW_PGSA_Scenarios {
   widgetClassPath=&/EUW_PGSA_Scenarios
   bIsVariable=true
   parent=WidgetSwitcher_0
   widget=EUW_PGSA_Scenarios
   slot {class=WidgetSwitcherSlot}
  }
  HorizontalBox_2 {widgetClassPath=&HorizontalBox, parent=StackBox_0, widget=HorizontalBox_2, slot {class=StackBoxSlot}}
  editTextName {
   widgetClassPath=&EditorUtilityEditableTextBox
   bIsVariable=true
   parent=HorizontalBox_2
   widget=editTextName
   properties {
    hintText=Scenario Name Here...
    minimumDesiredWidth=150
    toolTipText=Text here will be inserted into the scenario name in the text on the right.
   }
   slot {class=HorizontalBoxSlot, properties {padding=<ltrb 0 0 10 0>}}
  }
  editTextOutput {
   widgetClassPath=&EditorUtilityEditableTextBox
   bIsVariable=true
   parent=HorizontalBox_2
   widget=editTextOutput
   properties {isReadOnly=true}
   slot {class=HorizontalBoxSlot, properties {size {sizeRule=Fill, value=1}}}
  }
  btnCopyOutput {
   widgetClassPath=&TA_Generic01_Button
   parent=HorizontalBox_2
   widget=btnCopyOutput
   properties {label=Copy}
   slot {class=HorizontalBoxSlot}
  }
 ]
}
